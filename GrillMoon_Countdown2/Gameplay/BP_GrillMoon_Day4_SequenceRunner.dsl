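; Day 4 sequence runner for the GrillMoon countdown, which announces its day when initialized from the calendar event.
; It broadcasts and listens for a stateful GrillMoon day event through the global event router.
; The received day toggles equipment and "fun equipment" data layers, with certain states starting POI music and ambience or fast-forwarding this day to finished.
; Cheat events set a day percentage and run the server-side pre-load and play of the sequence.
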(event Custom|GM_Day4
  (params (Percentage :type Float%20%28double-precision%29))
  ;[0 0 0] Console Command / Load and Play
  (Variables|Cheats|SetCheat%20Day%20Percentage Percentage) ; Allows skipping ahead in the\nsequence for testing. A value\nbetween 0 - 1.0
  (CallFunction|SendDayEvent "NewEnumerator3")
  (if (Networking|IsServer)
    (Variables|Cheats|SetStarting%20Method "NewEnumerator1") ; Keeps some calendar logic from running
    (Variables|Default|SetRun%20From%20Cheat true)
    (CallFunction|%28CallOnServer%29Pre-LoadSequence)))

(event Custom|Set%20This%20Day%20to%20Finished
  ;[0 0 0] Console Command / Load and Play
  (Variables|Cheats|SetCheat%20Day%20Percentage 0.95) ; set to near finished so the mask lights up
  (Utilities|FlowControl|DoOnce false
    (:Completed
      (if (Networking|IsServer)
        (Variables|Cheats|SetStarting%20Method "NewEnumerator1") ; Keeps some calendar logic from running
        (Variables|Default|SetRun%20From%20Cheat true)
        (CallFunction|%28CallOnServer%29Pre-LoadSequence)))))

(event EventPre-LoadSequenceComplete
  ;[0 0 0] Console Command / Load and Play
  (|CallParentFunction)
  (if (Networking|IsServer)
    (if (Variables|Default|GetRun%20From%20Cheat)
      (CallFunction|LoadSequence) ; Play the sequence.
    )))

(event EventInitComplete-PreCalendarCheck
  ;[0 0 0] Console Command / Load and Play
  (|CallParentFunction)
  (CallFunction|BindDayEvent))

(event EventInitializefromCalendarEvent
  ;[0 0 0] Console Command / Load and Play
  (|CallParentFunction)
  (CallFunction|SendDayEvent "NewEnumerator3"))

(event Custom|Bind%20Day%20Event
  ;[0.3872 0.4056 1 0.5] Catch Event Router Events
  (bind (_asynctaskproxy _payload) (GameplayEventRouter|StartListeningToStatefulEvent (GameplayEventRouter|GetGlobalEventRouter) "/GrillMoon_Countdown/Gameplay/EventRouter/GrillMoon_Day_Event.GrillMoon_Day_Event")
    (:OnEventReceived
      _)
    (:OnSavedEventStateExists
      _))
  (switch Utilities|FlowControl|Switch|SwitchonGrillMoon_Day_Event_Enum (Utilities|Struct|BreakGrillMoonDayEvent _payload)
    (:NewEnumerator30
      _)
    (:NewEnumerator0
      _)
    (:NewEnumerator17
      _)
    (:NewEnumerator1
      _)
    (:NewEnumerator18
      _)
    (:NewEnumerator2
      _)
    (:NewEnumerator19
      (Class|BPGrillMoonPOIMusicComponent|StartPOIMusic (Variables|Default|GetBP_GrillMoon_POI_MusicComponent))
      (Class|BPGrillMoonPOIAmbienceComponent|SpawnPOIAmbience (Variables|Default|GetBP_GrillMoon_POI_AmbienceComponent))
      (bind _returnvalue (WorldPartition|GetDataLayerManager :!label _label))
      (DataLayers|SetDataLayerRuntimeState _returnvalue "/GrillMoon_Countdown2/DataLayers/DL_GrillMoon_Equipment.DL_GrillMoon_Equipment" "Activated" true :!class /Script/Engine.DataLayerManager) ; Equipment - Show
    )
    (:NewEnumerator3
      (goto _label))
    (:NewEnumerator20
      (CallFunction|SetThisDayToFinished) ; activate the mask quickly
      (goto _label))
    (:NewEnumerator4
      (goto _label))
    (:NewEnumerator21
      (Utilities|FlowControl|Sequence :!label _label_2
        (:then_0
          (goto _label))
        (:then_1
          (bind _returnvalue_2 (WorldPartition|GetDataLayerManager))
          (DataLayers|SetDataLayerRuntimeState _returnvalue_2 "/GrillMoon_Countdown2/DataLayers/DL_GrillMoon_Equipment2.DL_GrillMoon_Equipment2" "Activated" true :!class /Script/Engine.DataLayerManager) ; Equipment 2 "fun equipment" - Show
        )) ; show normal + fun equipment
    )
    (:NewEnumerator13
      (goto _label_2)))
  (bind _returnvalue_3 (WorldPartition|GetDataLayerManager)) ; hide equipment days 1 - 3
  (DataLayers|SetDataLayerRuntimeState _returnvalue_3 "/GrillMoon_Countdown2/DataLayers/DL_GrillMoon_Equipment.DL_GrillMoon_Equipment" "Unloaded" true :!class /Script/Engine.DataLayerManager) ; Equipment - Hide
  (DataLayers|SetDataLayerRuntimeState _returnvalue_3 "/GrillMoon_Countdown2/DataLayers/DL_GrillMoon_Equipment2.DL_GrillMoon_Equipment2" "Unloaded" true :!class /Script/Engine.DataLayerManager) ; Equipment 2 "fun equipment" - Hide
)

(event Custom|Send%20Day%20Event
  (params (Day :type GrillMoon_Day_Event_Enum%20Enum :default "NewEnumerator17"))
  ;[0.3872 0.4056 1 0.5] Catch Event Router Events
  (GameplayEventRouter|BroadcastStatefulEvent (GameplayEventRouter|GetGlobalEventRouter) (Utilities|Struct|MakeGrillMoonDayEvent Day) 0))

(fn ConstructionScript
  (|CallParentFunction))
